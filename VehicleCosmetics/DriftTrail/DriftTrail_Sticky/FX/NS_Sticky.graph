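# Niagara system for the Sticky drift trail vehicle cosmetic, drawing dark goo-like wraps, cones and streaming strands at the car's wheels.
# Layered wheel-wrap and cone meshes attach to skeletal mesh sockets, with curves wobbling the cones.
# Six tentacle ribbon emitters and a background ribbon stream backward, inheriting source movement, with quality-tiered spawn rates.
# User.offset shifts cones and ribbons, bIsFrontendPreview retunes ribbons for preview, and painted-variant overrides are disabled.
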NiagaraSystem /VehicleCosmetics/DriftTrail/DriftTrail_Sticky/FX/NS_Sticky
dependencies {
 usedDataInterfaces [
  &NiagaraDataInterfaceVectorCurve
  &NiagaraDataInterfaceArrayFloat3
  &NiagaraDataInterfaceCurve
  &NiagaraDataInterfaceSkeletalMesh
 ]
 usedDynamicInputs []
 usedModules [
  &self:SystemSpawnScript.SystemScriptSource.SystemScriptGraph.NiagaraNodeAssignment_0.SetVariables_#1
  &SystemState
 ]
}
emitters {
 Cone_FirstLayer {
  emitterUpdateScript [
   &EmitterState()
   &SpawnBurst_Instantaneous(Use Spawn Probability=false, UseLoopCountLimit=false) {module=SpawnBurst_Instantaneous001}
   &SpawnBurst_Instantaneous(Use Spawn Probability=false, UseLoopCountLimit=false)
   &SpawnRate(SpawnRate=50, Use Spawn Probability=false) {enabled=false}
  ]
  particleSpawnScript [
   &VC_PaintedVariantsOverride() {enabled=false}
   &InitializeParticle(
    Color=<rgba 0 0 0 5>
    Lifetime=$Emitter.CurrentLoopDuration
    Mesh Scale=&Multiply_VectorByFloat(Vector=<2.5 0.55 1>, Float=2)
    Mesh Scale Mode=::NewEnumerator4
    Sprite Rotation Mode=::NewEnumerator2
    Sprite Size Mode=::NewEnumerator3
    UsePositionOffset=false
    Write Material Random=true
   )
   &PartitionParticles()
   &SkeletalMeshLocation(
    Bone / Socket Index=&Add_Integer(A=$Particles.PartitionParticles.Partition, B=4)
    Bone / Socket Sampling Mode=::NewEnumerator5
    Interpolate Mesh Sampling=false
    MaskSpawnGroup=false
    Mesh Sampling Type=::NewEnumerator6
    Sampled Position Offset=&AddVectorToPosition(Position=<-60 0 -66>, Vector=$User.offset)
    Skeletal Mesh=$System.SkeletalMesh
   )
   &AddVelocity(Velocity=<-1500 0 0>) {enabled=false}
   &InitialMeshOrientation(Mesh Orientation Mode=::NewEnumerator2, Rotation=<0 0.05 0>, Use Rotation Vector=true)
  ]
  particleUpdateScript [
   &ParticleState()
   &ScaleMeshSize(Scale Factor=&VectorFromCurve(
     VectorCurve=DataInterface() {propertyValues {
       XCurve {keys [<at 0 1>,<at 1 1>]}
       YCurve {keys [<at 0 1>,<at 1 1>]}
       ZCurve {keys [
         <at 0 1 RCIM_Cubic>
         {arriveTangent=1, interpMode=RCIM_Cubic, leaveTangent=1, time=0.05, value=0.9}
         {arriveTangent=0.3333, interpMode=RCIM_Cubic, leaveTangent=0.3333, time=0.1, value=1.1}
         {arriveTangent=-0.6667, interpMode=RCIM_Cubic, leaveTangent=-0.6667, time=0.2, value=0.95}
         {arriveTangent=-1, interpMode=RCIM_Cubic, leaveTangent=-1, time=0.25, value=1}
         {arriveTangent=1, interpMode=RCIM_Cubic, leaveTangent=1, time=0.3, value=0.85}
         {arriveTangent=1, interpMode=RCIM_Cubic, leaveTangent=1, time=0.35, value=1.1}
         {arriveTangent=-0.6667, interpMode=RCIM_Cubic, leaveTangent=-0.6667, time=0.4, value=0.95}
         {arriveTangent=-0.3333, interpMode=RCIM_Cubic, leaveTangent=-0.3333, time=0.5, value=1}
         {arriveTangent=0.6667, interpMode=RCIM_Cubic, leaveTangent=0.6667, time=0.55, value=0.9}
         {arriveTangent=0.3333, interpMode=RCIM_Cubic, leaveTangent=0.3333, time=0.65, value=1.1}
         {arriveTangent=-0.5, interpMode=RCIM_Cubic, leaveTangent=-0.5, time=0.7, value=0.95}
         {arriveTangent=-1, interpMode=RCIM_Cubic, leaveTangent=-1, time=0.75, value=1.05}
         {arriveTangent=0.25, interpMode=RCIM_Cubic, leaveTangent=0.25, time=0.8, value=0.85}
         {arriveTangent=0.75, interpMode=RCIM_Cubic, leaveTangent=0.75, time=0.95, value=1.1}
         <at 1 1 RCIM_Cubic>
        ]}
      }}
     Scale Curve=&SelectVectorFromArray(
      Direct Array Index=$Particles.PartitionParticles.Partition
      Vector Selection Array=DataInterface() {propertyValues {FloatData [<1 1 1.2>, <1 -1 1.2>]}}
     )
    )) {module=ScaleMeshSize001}
   &ScaleMeshSize(Scale Factor=&SelectVectorFromArray(
     Direct Array Index=$Particles.PartitionParticles.Partition
     Vector Selection Array=DataInterface() {propertyValues {FloatData [<1 1 1>, <1 -1 1>]}}
    )) {enabled=false}
   &SolveForcesAndVelocity(Clamp Velocity=false, Limit Acceleration=false)
  ]
  rendererClasses [&NiagaraMeshRendererProperties, &NiagaraSpriteRendererProperties]
  renderers [
   {rendererClass=&NiagaraMeshRendererProperties, rendererIndex=0}
   {rendererClass=&NiagaraSpriteRendererProperties, rendererIndex=1}
  ]
 }
 Cone_SecondLayer {
  emitterUpdateScript [
   &EmitterState()
   &SpawnBurst_Instantaneous(Use Spawn Probability=false, UseLoopCountLimit=false) {module=SpawnBurst_Instantaneous001}
   &SpawnBurst_Instantaneous(Use Spawn Probability=false, UseLoopCountLimit=false)
   &SpawnRate(SpawnRate=4, Use Spawn Probability=false) {enabled=false}
  ]
  particleSpawnScript [
   &VC_PaintedVariantsOverride() {enabled=false}
   &InitializeParticle(
    Color=<rgba 0.2708 0.2708 0.2327 5>
    Lifetime=$Emitter.CurrentLoopDuration
    Mesh Scale=&Multiply_VectorByFloat(Vector=<2 0.55 1.2>, Float=2)
    Mesh Scale Mode=::NewEnumerator4
    Sprite Rotation Mode=::NewEnumerator2
    Sprite Size Mode=::NewEnumerator3
    UsePositionOffset=false
    Write Material Random=true
   )
   &PartitionParticles()
   &SkeletalMeshLocation(
    Bone / Socket Index=&Add_Integer(A=$Particles.PartitionParticles.Partition, B=4)
    Bone / Socket Sampling Mode=::NewEnumerator5
    Interpolate Mesh Sampling=false
    MaskSpawnGroup=false
    Mesh Sampling Type=::NewEnumerator6
    Sampled Position Offset=&AddVectorToPosition(
     Position=&SelectVectorFromArray(
      Direct Array Index=$Particles.PartitionParticles.Partition
      Vector Selection Array=DataInterface() {propertyValues {FloatData [<-68 0 -62.809>, <-68 0 -62.809>]}}
     )
     Vector=$User.offset
    )
    Skeletal Mesh=$System.SkeletalMesh
   )
   &AddVelocity(Velocity=<-1500 0 0>) {enabled=false}
   &InitialMeshOrientation(Mesh Orientation Mode=::NewEnumerator2, Rotation=<0 0.04 0>, Use Rotation Vector=true)
  ]
  particleUpdateScript [
   &ParticleState()
   &UpdateMeshOrientation(
    Rotation Rate=&FloatFromCurve(FloatCurve=DataInterface() {propertyValues {Curve {keys [<at 0 0>,<at 0.2 -0.125>,<at 0.3 0.5>,<at 0.4 -0.375>,<at 0.6 0.25>,<at 0.7 -0.2>,<at 0.8 0.3>,<at 0.9 -0.2>,<at 1 -0.5>]}}})
    Rotation Vector=<0 0.05 0>
   )
   &ScaleMeshSize(Scale Factor=&VectorFromCurve(
     VectorCurve=DataInterface() {propertyValues {
       XCurve {keys [<at 0 1>,<at 1 1>]}
       YCurve {keys [<at 0 1>,<at 1 1>]}
       ZCurve {keys [
         <at 0 1 RCIM_Cubic>
         {arriveTangent=1, interpMode=RCIM_Cubic, leaveTangent=1, time=0.05, value=0.9}
         {arriveTangent=0.3333, interpMode=RCIM_Cubic, leaveTangent=0.3333, time=0.1, value=1.1}
         {arriveTangent=-0.6667, interpMode=RCIM_Cubic, leaveTangent=-0.6667, time=0.2, value=0.95}
         {arriveTangent=-1, interpMode=RCIM_Cubic, leaveTangent=-1, time=0.25, value=1}
         {arriveTangent=1, interpMode=RCIM_Cubic, leaveTangent=1, time=0.3, value=0.85}
         {arriveTangent=1, interpMode=RCIM_Cubic, leaveTangent=1, time=0.35, value=1.1}
         {arriveTangent=-0.6667, interpMode=RCIM_Cubic, leaveTangent=-0.6667, time=0.4, value=0.95}
         {arriveTangent=-0.3333, interpMode=RCIM_Cubic, leaveTangent=-0.3333, time=0.5, value=1}
         {arriveTangent=0.6667, interpMode=RCIM_Cubic, leaveTangent=0.6667, time=0.55, value=0.9}
         {arriveTangent=0.3333, interpMode=RCIM_Cubic, leaveTangent=0.3333, time=0.65, value=1.1}
         {arriveTangent=-0.5, interpMode=RCIM_Cubic, leaveTangent=-0.5, time=0.7, value=0.95}
         {arriveTangent=-1, interpMode=RCIM_Cubic, leaveTangent=-1, time=0.75, value=1.05}
         {arriveTangent=0.25, interpMode=RCIM_Cubic, leaveTangent=0.25, time=0.8, value=0.85}
         {arriveTangent=0.75, interpMode=RCIM_Cubic, leaveTangent=0.75, time=0.95, value=1.1}
         <at 1 1 RCIM_Cubic>
        ]}
      }}
     Scale Curve=&SelectVectorFromArray(
      Direct Array Index=$Particles.PartitionParticles.Partition
      Vector Selection Array=DataInterface() {propertyValues {FloatData [<1 1 1>, <1 1 1>]}}
     )
    )) {module=ScaleMeshSize001}
   &SolveForcesAndVelocity(Clamp Velocity=false, Limit Acceleration=false)
   &ScaleColor(Scale RGB=<0.13 0.13 0.13>, ScaleA=true, ScaleRGB=true)
  ]
  rendererClasses [&NiagaraMeshRendererProperties, &NiagaraSpriteRendererProperties]
  renderers [
   {rendererClass=&NiagaraMeshRendererProperties, rendererIndex=0}
   {rendererClass=&NiagaraSpriteRendererProperties, rendererIndex=1}
  ]
 }
 Ribbon_Background001 {
  emitterUpdateScript [
   &VC_RateQualityHandler(Epic Spawn Rate=52, High Spawn Rate=60, Low Spawn Rate=30, Medium Spawn Rate=35, Use Custom Rate?=true)
   &EmitterState(
    Inactive Response=::NewEnumerator1
    Life Cycle Mode=::NewEnumerator1
    Loop Duration=0.5
    Recalculate Duration Each Loop=true
    UseLoopDelay=false
   )
   &VC_SpawnRibbon_v1(Number of Ribbons=2, SpawnRate=$Emitter.Spawn Rate)
  ]
  particleSpawnScript [
   &VC_PaintedVariantsOverride() {enabled=false}
   &VC_InitializeRibbon_v1(
    Color=<rgba 0 0 0 1>
    Custom Offset=true
    Enable Skeletal Position=true
    Facing Axis:ENiagaraOrientationAxis=::ZAxis
    Flip Facing=true
    Local Vector Offset=&MakeCustomVectorFromBool(
     Bool=$User.bIsFrontendPreview
     True Vector=&AddVectorToPosition(Position=<-55 0 -37>, Vector=$User.offset)
     False Vector=&AddVectorToPosition(Position=<-90 0 -35>, Vector=$User.offset)
    )
    Offset Amount=1
    Ribbon Width:NiagaraFloat=&Multiply_Float(A=25, B=5)
   )
   &AddVelocity(Velocity=&MakeCustomVectorFromBool(Bool=$User.bIsFrontendPreview, True Vector=<-2500 0 70>, False Vector=<-2500 0 0>)) {module=AddVelocity001}
  ]
  particleUpdateScript [
   &ParticleState()
   &InheritSourceMovement(
    Applied Position Scale=&Multiply_VectorByFloat(
     Vector=<1 1 1>
     Float=&FloatFromCurve(FloatCurve=DataInterface() {propertyValues {Curve {keys [
          <at 0 1 RCIM_Cubic>
          {arriveTangent=-0.1442, interpMode=RCIM_Cubic, leaveTangent=-0.1442, tangentMode=RCTM_User, time=0.6, value=1}
          {arriveTangent=-3.4668, interpMode=RCIM_Cubic, leaveTangent=-3.4668, tangentMode=RCTM_User, time=1, value=0}
         ]}}})
    )
    ApplyForce=false
    ApplyPosition=true
    ApplyPositionFalloff=false
   ) {module=InheritSourceMovement001}
   &SolveForcesAndVelocity(Clamp Velocity=false, Limit Acceleration=false)
   &VC_UpdateRibbon_v1(Enable Skeletal Position=true)
   &ScaleRibbonWidth(Ribbon Width Scale=&FloatFromCurve(FloatCurve=DataInterface() {propertyValues {Curve {keys [
         {interpMode=RCIM_Cubic, tangentMode=RCTM_Break, time=0, value=1}
         {arriveTangent=1.6788, interpMode=RCIM_Cubic, leaveTangent=1.4514, tangentMode=RCTM_Break, time=0.5, value=1.525}
         {arriveTangent=1.4473, tangentMode=RCTM_User, time=1, value=2.5}
        ]}}})) {enabled=false}
   &DynamicMaterialParameters(
    Index 0 Param 1=&MakeCustomFloatFromBool(Bool=$User.bIsFrontendPreview, True Float=1.5, False Float=1)
    Index 0 Param 2=1
    Index 0 Param 3=3.5
    Module.Index0Param0Write=true
    Module.Index0Param1Write=true
    Module.Index0Param2Write=true
    Module.Index0Param3Write=true
   )
   &ScaleColor(
    Scale Alpha=&FloatFromCurve(
     FloatCurve=DataInterface() {propertyValues {Curve {keys [
         {arriveTangent=-6.0703, interpMode=RCIM_Cubic, leaveTangent=-6.0703, tangentMode=RCTM_User, time=0.1, value=2}
         {arriveTangentWeight=0.6967, interpMode=RCIM_Cubic, tangentMode=RCTM_User, time=1, value=0}
        ]}}}
     Scale Curve=&FloatFromCurve(FloatCurve=DataInterface() {propertyValues {Curve {keys [<at 0 1>,<at 1 0>]}}}, Scale Curve=2)
    )
    ScaleA=true
    ScaleRGB=true
   )
   &ScaleRibbonWidth(Ribbon Width Scale=&FloatFromCurve(FloatCurve=DataInterface() {propertyValues {Curve {keys [{tangentMode=RCTM_User, time=0, value=1}, <at 0.5 1.2>, <at 0.7 2>]}}})) {module=ScaleRibbonWidth001}
  ]
  rendererClasses [&NiagaraRibbonRendererProperties]
  renderers [{rendererClass=&NiagaraRibbonRendererProperties}]
 }
 Tentacle01 {
  emitterUpdateScript [
   &VC_RateQualityHandler(Epic Spawn Rate=124, High Spawn Rate=100, Low Spawn Rate=62, Medium Spawn Rate=80, Use Custom Rate?=true)
   &EmitterState()
   &VC_SpawnRibbon_v1(Number of Ribbons=2, SpawnRate=$Emitter.Spawn Rate)
  ]
  particleSpawnScript [
   &VC_PaintedVariantsOverride() {enabled=false}
   &VC_InitializeRibbon_v1(
    Color=<rgba 0.3646 0.3646 0.3646 1>
    Custom Offset=true
    Enable Skeletal Position=true
    Lifetime=0.45
    Local Vector Offset=&MakeCustomVectorFromBool(
     Bool=$User.bIsFrontendPreview
     True Vector=&AddVectorToPosition(Position=<-60 7 -30>, Vector=$User.offset)
     False Vector=&AddVectorToPosition(Position=<-100 2.5 -35>, Vector=$User.offset)
    )
    Offset Amount=1
    Ribbon Width:NiagaraFloat=&Multiply_Float(A=25, B=1)
   )
   &AddVelocity(Velocity=&MakeCustomVectorFromBool(Bool=$User.bIsFrontendPreview, True Vector=<-2000 0 0>, False Vector=<-2500 0 0>)) {module=AddVelocity001}
  ]
  particleUpdateScript [
   &ParticleState()
   &InheritSourceMovement(
    Applied Position Scale=&Multiply_VectorByFloat(
     Vector=<1 1 1>
     Float=&FloatFromCurve(FloatCurve=DataInterface() {propertyValues {Curve {keys [
          <at 0 1 RCIM_Cubic>
          {arriveTangent=-0.1442, interpMode=RCIM_Cubic, leaveTangent=-0.1442, tangentMode=RCTM_User, time=0.6, value=1}
          {arriveTangent=-3.4668, interpMode=RCIM_Cubic, leaveTangent=-3.4668, tangentMode=RCTM_User, time=1, value=0}
         ]}}})
    )
    ApplyForce=false
    ApplyPosition=true
    ApplyPositionFalloff=false
   ) {module=InheritSourceMovement001}
   &SolveForcesAndVelocity(Clamp Velocity=false, Limit Acceleration=false)
   &VC_UpdateRibbon_v1(Enable Skeletal Position=true)
   &ScaleRibbonWidth(Ribbon Width Scale=&FloatFromCurve(FloatCurve=DataInterface() {propertyValues {Curve {keys [<at 0 0.05>,<at 0.2 1.2>,<at 0.5 1.2>,<at 1 0>]}}}))
   &DynamicMaterialParameters(
    Index 0 Param 1=1.1
    Index 0 Param 3=1
    Module.Index0Param0Write=true
    Module.Index0Param1Write=true
    Module.Index0Param2Write=true
    Module.Index0Param3Write=true
   )
   &ScaleColor(Scale RGB=<5.2 5.2 5.2>, ScaleA=true, ScaleRGB=true)
  ]
  rendererClasses [&NiagaraRibbonRendererProperties]
  renderers [{rendererClass=&NiagaraRibbonRendererProperties}]
 }
 Tentacle02 {
  emitterUpdateScript [
   &VC_RateQualityHandler(Epic Spawn Rate=280, High Spawn Rate=240, Low Spawn Rate=140, Medium Spawn Rate=200, Use Custom Rate?=true)
   &EmitterState()
   &VC_SpawnRibbon_v1(Number of Ribbons=2, SpawnRate=$Emitter.Spawn Rate)
  ]
  particleSpawnScript [
   &VC_PaintedVariantsOverride() {enabled=false}
   &VC_InitializeRibbon_v1(
    Color=<rgba 0.3646 0.3646 0.3646 1>
    Custom Offset=true
    Enable Skeletal Position=true
    Lifetime=0.2
    Local Vector Offset=&MakeCustomVectorFromBool(
     Bool=$User.bIsFrontendPreview
     True Vector=&AddVectorToPosition(Position=<-60 5 -30>, Vector=$User.offset)
     False Vector=&AddVectorToPosition(Position=<-60 5 -30>, Vector=$User.offset)
    )
    Offset Amount=1
    Ribbon Width:NiagaraFloat=&Multiply_Float(A=15, B=1)
   )
   &AddVelocity(Velocity=&SelectVectorFromArray(
     Direct Array Index=$Emitter.RibbonCount
     Vector Selection Array=DataInterface() {propertyValues {FloatData [<-2000 -500 250>, <-2000 500 250>]}}
    )) {module=AddVelocity001}
  ]
  particleUpdateScript [
   &ParticleState()
   &InheritSourceMovement(
    Applied Position Scale=&Multiply_VectorByFloat(
     Vector=<1 1 1>
     Float=&FloatFromCurve(FloatCurve=DataInterface() {propertyValues {Curve {keys [
          <at 0 1 RCIM_Cubic>
          {arriveTangent=-0.1442, interpMode=RCIM_Cubic, leaveTangent=-0.1442, tangentMode=RCTM_User, time=0.6, value=1}
          {arriveTangent=-3.4668, interpMode=RCIM_Cubic, leaveTangent=-3.4668, tangentMode=RCTM_User, time=1, value=0}
         ]}}})
    )
    ApplyForce=false
    ApplyPosition=true
    ApplyPositionFalloff=false
   ) {module=InheritSourceMovement001}
   &SolveForcesAndVelocity(Clamp Velocity=false, Limit Acceleration=false)
   &VC_UpdateRibbon_v1(Enable Skeletal Position=true)
   &ScaleRibbonWidth(Ribbon Width Scale=&FloatFromCurve(FloatCurve=DataInterface() {propertyValues {Curve {keys [<at 0 1>,<at 0.1 1.2>,<at 0.5 1.2>,<at 1 0>]}}}, Scale Curve=1.5))
   &DynamicMaterialParameters(
    Index 0 Param 1=1.1
    Module.Index0Param0Write=true
    Module.Index0Param1Write=true
    Module.Index0Param2Write=true
    Module.Index0Param3Write=true
   )
   &ScaleColor(Scale RGB=<5.2 5.2 5.2>, ScaleA=true, ScaleRGB=true)
  ]
  rendererClasses [&NiagaraRibbonRendererProperties]
  renderers [{rendererClass=&NiagaraRibbonRendererProperties}]
 }
 Tentacle03 {
  emitterUpdateScript [
   &VC_RateQualityHandler(Epic Spawn Rate=200, High Spawn Rate=180, Low Spawn Rate=100, Medium Spawn Rate=140, Use Custom Rate?=true)
   &EmitterState()
   &VC_SpawnRibbon_v1(Number of Ribbons=2, SpawnRate=$Emitter.Spawn Rate)
  ]
  particleSpawnScript [
   &VC_PaintedVariantsOverride() {enabled=false}
   &VC_InitializeRibbon_v1(
    Color=<rgba 0.3646 0.3646 0.3646 1>
    Custom Offset=true
    Enable Skeletal Position=true
    Lifetime=0.18
    Local Vector Offset=&MakeCustomVectorFromBool(
     Bool=$User.bIsFrontendPreview
     True Vector=&AddVectorToPosition(Position=<-65 -5 -30>, Vector=$User.offset)
     False Vector=&AddVectorToPosition(Position=<-65 -2 -35>, Vector=$User.offset)
    )
    Offset Amount=1
    Ribbon Width:NiagaraFloat=&Multiply_Float(A=15, B=1)
   )
   &AddVelocity(Velocity=<-2000 -200 200>) {module=AddVelocity001}
  ]
  particleUpdateScript [
   &ParticleState()
   &InheritSourceMovement(
    Applied Position Scale=&Multiply_VectorByFloat(
     Vector=<1 1 1>
     Float=&FloatFromCurve(FloatCurve=DataInterface() {propertyValues {Curve {keys [
          <at 0 1 RCIM_Cubic>
          {arriveTangent=-0.1442, interpMode=RCIM_Cubic, leaveTangent=-0.1442, tangentMode=RCTM_User, time=0.6, value=1}
          {arriveTangent=-3.4668, interpMode=RCIM_Cubic, leaveTangent=-3.4668, tangentMode=RCTM_User, time=1, value=0}
         ]}}})
    )
    ApplyForce=false
    ApplyPosition=true
    ApplyPositionFalloff=false
   ) {module=InheritSourceMovement001}
   &SolveForcesAndVelocity(Clamp Velocity=false, Limit Acceleration=false)
   &VC_UpdateRibbon_v1(Enable Skeletal Position=true)
   &ScaleRibbonWidth(Ribbon Width Scale=&FloatFromCurve(FloatCurve=DataInterface() {propertyValues {Curve {keys [<at 0 1>,<at 0.1 1.2>,<at 0.5 1.2>,<at 1 0>]}}}, Scale Curve=1.5))
   &DynamicMaterialParameters(
    Index 0 Param 1=1.1
    Module.Index0Param0Write=true
    Module.Index0Param1Write=true
    Module.Index0Param2Write=true
    Module.Index0Param3Write=true
   )
   &ScaleColor(Scale RGB=<5.2 5.2 5.2>, ScaleA=true, ScaleRGB=true)
  ]
  rendererClasses [&NiagaraRibbonRendererProperties]
  renderers [{rendererClass=&NiagaraRibbonRendererProperties}]
 }
 Tentacle04 {
  emitterUpdateScript [
   &VC_RateQualityHandler(Epic Spawn Rate=112, High Spawn Rate=100, Low Spawn Rate=66, Medium Spawn Rate=80, Use Custom Rate?=true)
   &EmitterState()
   &VC_SpawnRibbon_v1(Number of Ribbons=2, SpawnRate=$Emitter.Spawn Rate)
  ]
  particleSpawnScript [
   &VC_PaintedVariantsOverride() {enabled=false}
   &VC_InitializeRibbon_v1(
    Color=<rgba 0.35 0.35 0.3008 1>
    Custom Offset=true
    Enable Skeletal Position=true
    Lifetime=0.45
    Local Vector Offset=&MakeCustomVectorFromBool(
     Bool=$User.bIsFrontendPreview
     True Vector=&AddVectorToPosition(Position=<-60 5 -30>, Vector=$User.offset)
     False Vector=&AddVectorToPosition(Position=<-100 -2.5 -30>, Vector=$User.offset)
    )
    Offset Amount=1
    Ribbon Width:NiagaraFloat=&Multiply_Float(A=20, B=1)
   )
   &AddVelocity(Velocity=&MakeCustomVectorFromBool(Bool=$User.bIsFrontendPreview, True Vector=<-2000 0 0>, False Vector=<-2500 0 0>)) {module=AddVelocity001}
  ]
  particleUpdateScript [
   &ParticleState()
   &InheritSourceMovement(
    Applied Position Scale=&Multiply_VectorByFloat(
     Vector=<1 1 1>
     Float=&FloatFromCurve(FloatCurve=DataInterface() {propertyValues {Curve {keys [
          <at 0 1 RCIM_Cubic>
          {arriveTangent=-0.1442, interpMode=RCIM_Cubic, leaveTangent=-0.1442, tangentMode=RCTM_User, time=0.6, value=1}
          {arriveTangent=-3.4668, interpMode=RCIM_Cubic, leaveTangent=-3.4668, tangentMode=RCTM_User, time=1, value=0}
         ]}}})
    )
    ApplyForce=false
    ApplyPosition=true
    ApplyPositionFalloff=false
   ) {module=InheritSourceMovement001}
   &SolveForcesAndVelocity(Clamp Velocity=false, Limit Acceleration=false)
   &VC_UpdateRibbon_v1(Enable Skeletal Position=true)
   &ScaleRibbonWidth(Ribbon Width Scale=&FloatFromCurve(FloatCurve=DataInterface() {propertyValues {Curve {keys [<at 0 0.05>,<at 0.2 1>,<at 0.5 1.2>,<at 1 0>]}}}, Scale Curve=1.25))
   &DynamicMaterialParameters(
    Index 0 Param 1=1.1
    Index 0 Param 3=1
    Module.Index0Param0Write=true
    Module.Index0Param1Write=true
    Module.Index0Param2Write=true
    Module.Index0Param3Write=true
   )
   &ScaleColor(Scale RGB=<5.2 5.2 5.2>, ScaleA=true, ScaleRGB=true)
  ]
  rendererClasses [&NiagaraRibbonRendererProperties]
  renderers [{rendererClass=&NiagaraRibbonRendererProperties}]
 }
 Tentacle05 {
  emitterUpdateScript [
   &VC_RateQualityHandler(Epic Spawn Rate=110, High Spawn Rate=100, Low Spawn Rate=55, Medium Spawn Rate=80, Use Custom Rate?=true)
   &EmitterState()
   &VC_SpawnRibbon_v1(Number of Ribbons=2, SpawnRate=$Emitter.Spawn Rate)
  ]
  particleSpawnScript [
   &VC_PaintedVariantsOverride() {enabled=false}
   &VC_InitializeRibbon_v1(
    Color=<rgba 0.3646 0.3646 0.3646 1>
    Custom Offset=true
    Enable Skeletal Position=true
    Lifetime=0.44
    Local Vector Offset=&MakeCustomVectorFromBool(
     Bool=$User.bIsFrontendPreview
     True Vector=&AddVectorToPosition(Position=<-60 -5 -28>, Vector=$User.offset)
     False Vector=&AddVectorToPosition(Position=<-100 2.5 -30>, Vector=$User.offset)
    )
    Offset Amount=1
    Ribbon Width:NiagaraFloat=&Multiply_Float(A=25, B=1)
   )
   &AddVelocity(Velocity=&MakeCustomVectorFromBool(Bool=$User.bIsFrontendPreview, True Vector=<-2000 0 50>, False Vector=<-2500 0 0>)) {module=AddVelocity001}
  ]
  particleUpdateScript [
   &ParticleState()
   &InheritSourceMovement(
    Applied Position Scale=&Multiply_VectorByFloat(
     Vector=<1 1 1>
     Float=&FloatFromCurve(FloatCurve=DataInterface() {propertyValues {Curve {keys [
          <at 0 1 RCIM_Cubic>
          {arriveTangent=-0.1442, interpMode=RCIM_Cubic, leaveTangent=-0.1442, tangentMode=RCTM_User, time=0.6, value=1}
          {arriveTangent=-3.4668, interpMode=RCIM_Cubic, leaveTangent=-3.4668, tangentMode=RCTM_User, time=1, value=0}
         ]}}})
    )
    ApplyForce=false
    ApplyPosition=true
    ApplyPositionFalloff=false
   ) {module=InheritSourceMovement001}
   &SolveForcesAndVelocity(Clamp Velocity=false, Limit Acceleration=false)
   &VC_UpdateRibbon_v1(Enable Skeletal Position=true)
   &ScaleRibbonWidth(Ribbon Width Scale=&FloatFromCurve(FloatCurve=DataInterface() {propertyValues {Curve {keys [<at 0 0.05>,<at 0.2 1.2>,<at 0.5 1.2>,<at 1 0>]}}}, Scale Curve=1.25))
   &DynamicMaterialParameters(
    Index 0 Param 1=1.1
    Index 0 Param 3=1
    Module.Index0Param0Write=true
    Module.Index0Param1Write=true
    Module.Index0Param2Write=true
    Module.Index0Param3Write=true
   )
   &ScaleColor(Scale RGB=<5.2 5.2 5.2>, ScaleA=true, ScaleRGB=true)
  ]
  rendererClasses [&NiagaraRibbonRendererProperties]
  renderers [{rendererClass=&NiagaraRibbonRendererProperties}]
 }
 Tentacle06 {
  emitterUpdateScript [
   &VC_RateQualityHandler(Epic Spawn Rate=108, High Spawn Rate=94, Low Spawn Rate=54, Medium Spawn Rate=64, Use Custom Rate?=true)
   &EmitterState()
   &VC_SpawnRibbon_v1(Number of Ribbons=2, SpawnRate=$Emitter.Spawn Rate)
  ]
  particleSpawnScript [
   &VC_PaintedVariantsOverride() {enabled=false}
   &VC_InitializeRibbon_v1(
    Color=<rgba 0.3646 0.3646 0.3646 1>
    Custom Offset=true
    Enable Skeletal Position=true
    Lifetime=0.52
    Local Vector Offset=&MakeCustomVectorFromBool(
     Bool=$User.bIsFrontendPreview
     True Vector=&AddVectorToPosition(Position=<-60 -5 -29>, Vector=$User.offset)
     False Vector=&AddVectorToPosition(Position=<-100 2.5 -32>, Vector=$User.offset)
    )
    Offset Amount=1
    Ribbon Width:NiagaraFloat=&Multiply_Float(A=17, B=1)
   )
   &AddVelocity(Velocity=&MakeCustomVectorFromBool(Bool=$User.bIsFrontendPreview, True Vector=<-2000 0 0>, False Vector=<-2500 0 0>)) {module=AddVelocity001}
  ]
  particleUpdateScript [
   &ParticleState()
   &InheritSourceMovement(
    Applied Position Scale=&Multiply_VectorByFloat(
     Vector=<1 1 1>
     Float=&FloatFromCurve(FloatCurve=DataInterface() {propertyValues {Curve {keys [
          <at 0 1 RCIM_Cubic>
          {arriveTangent=-0.1442, interpMode=RCIM_Cubic, leaveTangent=-0.1442, tangentMode=RCTM_User, time=0.6, value=1}
          {arriveTangent=-3.4668, interpMode=RCIM_Cubic, leaveTangent=-3.4668, tangentMode=RCTM_User, time=1, value=0}
         ]}}})
    )
    ApplyForce=false
    ApplyPosition=true
    ApplyPositionFalloff=false
   ) {module=InheritSourceMovement001}
   &SolveForcesAndVelocity(Clamp Velocity=false, Limit Acceleration=false)
   &VC_UpdateRibbon_v1(Enable Skeletal Position=true)
   &ScaleRibbonWidth(Ribbon Width Scale=&FloatFromCurve(FloatCurve=DataInterface() {propertyValues {Curve {keys [<at 0 0.05>,<at 0.2 1>,<at 0.5 1.2>,<at 1 0>]}}})) {enabled=false}
   &DynamicMaterialParameters(
    Index 0 Param 1=1.1
    Index 0 Param 3=1
    Module.Index0Param0Write=true
    Module.Index0Param1Write=true
    Module.Index0Param2Write=true
    Module.Index0Param3Write=true
   )
   &ScaleColor(Scale RGB=<5.2 5.2 5.2>, ScaleA=true, ScaleRGB=true)
   &ScaleRibbonWidth(Ribbon Width Scale=&FloatFromCurve(FloatCurve=DataInterface() {propertyValues {Curve {keys [<at 0 1>,<at 0.2 1.2>,<at 0.5 1.2>,<at 1 0>]}}}, Scale Curve=1.5)) {module=ScaleRibbonWidth001}
  ]
  rendererClasses [&NiagaraRibbonRendererProperties]
  renderers [{rendererClass=&NiagaraRibbonRendererProperties}]
 }
 WheelWrap_FirstLayer {
  emitterUpdateScript [
   &EmitterState(Life Cycle Mode=::NewEnumerator1, Loop Behavior=::NewEnumerator1, Loop Duration=1, UseLoopDelay=false)
   &SpawnBurst_Instantaneous(Use Spawn Probability=false, UseLoopCountLimit=false) {module=SpawnBurst_Instantaneous001}
   &SpawnBurst_Instantaneous(Use Spawn Probability=false, UseLoopCountLimit=false)
   &VC_WheelWrapScale()
  ]
  particleSpawnScript [
   &VC_PaintedVariantsOverride() {enabled=false}
   &InitializeParticle(
    Color Mode=::NewEnumerator0
    Mesh Scale=$Emitter.WheelFXScale
    Mesh Scale Mode=::NewEnumerator4
    Sprite Rotation Mode=::NewEnumerator2
    Sprite Size Mode=::NewEnumerator3
    UsePositionOffset=false
    Write Material Random=true
   )
   &PartitionParticles()
   &SkeletalMeshLocation(
    Bone / Socket Index=$Particles.PartitionParticles.Partition
    Bone / Socket Sampling Mode=::NewEnumerator5
    Interpolate Mesh Sampling=false
    MaskSpawnGroup=false
    Mesh Sampling Type=::NewEnumerator6
    Skeletal Mesh=$System.SkeletalMesh
   )
   &Color(Color=<rgba 0 0 0 3>)
   &InitialMeshOrientation(Mesh Orientation Mode=::NewEnumerator2, Rotation Coordinate Space=::NewEnumerator2, Use Rotation Vector=true)
  ]
  particleUpdateScript [
   &ParticleState(Kill Particles When Lifetime Has Elapsed=false)
   &SolveForcesAndVelocity(Clamp Velocity=false, Limit Acceleration=false)
  ]
  rendererClasses [&NiagaraMeshRendererProperties]
  renderers [{rendererClass=&NiagaraMeshRendererProperties, rendererIndex=0}]
 }
 WheelWrap_SecondLayer {
  emitterUpdateScript [
   &EmitterState(Life Cycle Mode=::NewEnumerator1, Loop Behavior=::NewEnumerator1, Loop Duration=1, UseLoopDelay=false)
   &SpawnBurst_Instantaneous(Use Spawn Probability=false, UseLoopCountLimit=false) {module=SpawnBurst_Instantaneous001}
   &SpawnBurst_Instantaneous(Use Spawn Probability=false, UseLoopCountLimit=false)
   &VC_WheelWrapScale()
  ]
  particleSpawnScript [
   &VC_PaintedVariantsOverride() {enabled=false}
   &InitializeParticle(
    Color Mode=::NewEnumerator0
    Mesh Scale=$Emitter.WheelFXScale
    Mesh Scale Mode=::NewEnumerator4
    Sprite Rotation Mode=::NewEnumerator2
    Sprite Size Mode=::NewEnumerator3
    UsePositionOffset=false
    Write Material Random=true
   )
   &PartitionParticles()
   &SkeletalMeshLocation(
    Bone / Socket Index=$Particles.PartitionParticles.Partition
    Bone / Socket Sampling Mode=::NewEnumerator5
    Interpolate Mesh Sampling=false
    MaskSpawnGroup=false
    Mesh Sampling Type=::NewEnumerator6
    Skeletal Mesh=$System.SkeletalMesh
   )
   &Color(Color=<rgba 0.04 0.04 0.0344 3>)
   &InitialMeshOrientation(Mesh Orientation Mode=::NewEnumerator2, Rotation Coordinate Space=::NewEnumerator2, Use Rotation Vector=true)
  ]
  particleUpdateScript [
   &ParticleState(Kill Particles When Lifetime Has Elapsed=false)
   &SolveForcesAndVelocity(Clamp Velocity=false, Limit Acceleration=false)
  ]
  rendererClasses [&NiagaraMeshRendererProperties]
  renderers [{rendererClass=&NiagaraMeshRendererProperties, rendererIndex=0}]
 }
}
summary {
 systemName=NS_Sticky
 emitters [
  {
   emitterName=Cone_FirstLayer
   rendererClasses [&NiagaraMeshRendererProperties, &NiagaraSpriteRendererProperties]
  }
  {
   emitterName=Cone_SecondLayer
   rendererClasses [&NiagaraMeshRendererProperties, &NiagaraSpriteRendererProperties]
  }
  {emitterName=Tentacle01, rendererClasses [&NiagaraRibbonRendererProperties]}
  {emitterName=Tentacle02, rendererClasses [&NiagaraRibbonRendererProperties]}
  {emitterName=Tentacle03, rendererClasses [&NiagaraRibbonRendererProperties]}
  {emitterName=Tentacle04, rendererClasses [&NiagaraRibbonRendererProperties]}
  {emitterName=Tentacle05, rendererClasses [&NiagaraRibbonRendererProperties]}
  {emitterName=Tentacle06, rendererClasses [&NiagaraRibbonRendererProperties]}
  {emitterName=WheelWrap_FirstLayer, rendererClasses [&NiagaraMeshRendererProperties]}
  {emitterName=WheelWrap_SecondLayer, rendererClasses [&NiagaraMeshRendererProperties]}
  {emitterName=Ribbon_Background001, rendererClasses [&NiagaraRibbonRendererProperties]}
 ]
 userVariables [
  User.bIsFrontendPreview {type=NiagaraBool}
  User.bIsPainted {type=NiagaraBool}
  User.ColorVariantID {type=NiagaraInt32}
  User.offset {type=Vector3f, description="offset from wheels(set through cosmetic template)", defaultValue {struct=Vector3f, value=<0 0 0>}}
  User.PaintedColor {type=LinearColor, description="", defaultValue {struct=LinearColor, value=<rgba 1 1 1 1>}}
  User.PaintedScalar {type=NiagaraFloat, defaultValue=1, description=""}
  User.WheelSize {type=NiagaraFloat, defaultValue=0.1, description=""}
 ]
}
