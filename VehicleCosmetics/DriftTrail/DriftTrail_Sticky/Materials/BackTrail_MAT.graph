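# Unlit translucent Niagara material for the back trail of the Sticky drift trail vehicle cosmetic.
# Panning cloud noise distorts a scrolling trail texture, with strength cycling between two Distortion parameters over time.
# Hard-thresholded trail alpha, particle alpha, panning dark-cloud and rotated foam gradients, and a U-axis FadeIn ramp multiply into a saturated mask.
# The thresholded, sharpened red channel tints particle color, feeding the shared emissive adjustment function.
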Material /VehicleCosmetics/DriftTrail/DriftTrail_Sticky/Materials/BackTrail_MAT
graph {
 properties {
  bAutomaticallySetUsageInEditor=false
  bUsedWithNiagaraMeshParticles=true
  bUsedWithNiagaraRibbons=true
  bUsedWithNiagaraSprites=true
  blendMode=BLEND_Translucent
  shadingModel=MSM_Unlit
  usedShadingModels=Unlit
 }
 nodes [
  MaterialExpressionVertexColor_0 = ParticleColor()
  n1 = Panner(Coordinate=TextureCoordinate() {uTiling=0.35, vTiling=0.35}, Time=Multiply(A=Time(), B=Constant() {r=0.2})) {speedX=-1.5}
  n2 = Multiply(
   A=ComponentMask(None=TextureSample(UVs=@n1) {texture=/DealershipSharedLibrary/FX/Textures/Noise/T_VCVFX_CloudsSoft_RGB}.RGB) {g=true, r=true}
   B=LinearInterpolate(
    A=ScalarParameter() {defaultValue=0.15, parameterName=Distortion}
    B=ScalarParameter() {defaultValue=0.1, parameterName=Distortion_2}
    Alpha=Time() {bOverride_Period=true, period=0.5}
   )
  )
  n3 = TextureCoordinate() {uTiling=0.5}
  n4 = TextureSample(UVs=Panner(Coordinate=Add(A=@n2, B=@n3)) {speedX=-1}) {texture=/DealershipSharedLibrary/FX/Textures/Gradient_Masks/T_VCVFX_Sticky_BackTrail_A}
  n5 = Multiply(A=@n4.A)
  n6 = Multiply(A=@MaterialExpressionVertexColor_0.A, B=Power(Base=SmoothStep(Value=@n5) {constMin=0.94}, Exp=Constant() {r=1}))
  n7 = ScalarParameter() {defaultValue=1, parameterName=X}
  n8 = ScalarParameter() {defaultValue=1, parameterName=Y}
  n9 = Panner(Coordinate=Multiply(A=TextureCoordinate() {uTiling=2, vTiling=2}, B=AppendVector(A=@n7, B=@n8))) {speedX=-1.5}
  n10 = Panner(Coordinate=MaterialFunctionCall(UVs=TextureCoordinate() {vTiling=2}, "Rotation Angle (0-1)"=Constant() {r=0.25}) {materialFunction=&CustomRotator}.Rotated Values) {speedX=-0.3}
  n11 = Multiply(
   A=Multiply(A=TextureSample(UVs=@n9) {texture=/DealershipSharedLibrary/FX/Textures/Gradient_Masks/T_VCVFX_DarkcloudTail_A}.R, B=Constant() {r=4})
   B=Power(
    Base=TextureSample(UVs=@n10) {texture=/DealershipSharedLibrary/FX/Textures/Gradient_Masks/T_VCVFX_StickyFoamGradient01_A}.R
    Exp=Constant() {r=2}
   )
  )
  n12 = Constant()
  n13 = TextureCoordinate()
  n14 = Multiply(A=Power(Base=ComponentMask(None=@n13) {r=true}, Exp=Constant() {r=1.25}), B=Constant() {r=13})
  n15 = ScalarParameter() {defaultValue=1, parameterName=FadeIn}
  MaterialExpressionConstantClamp_0 = Saturate(None=Multiply(A=Multiply(A=@n6, B=@n11), B=SmoothStep(Min=@n12, Max=@n15, Value=@n14)))
  n16 = SmoothStep(Value=@n4.R) {constMin=0.95}
  n17 = Constant() {r=20}
  MaterialFunctionCall(Emissive=Multiply(A=@MaterialExpressionVertexColor_0.RGB, B=Power(Base=@n16, Exp=@n17))) {materialFunction=&MF_FX_EmissiveAdjustments}
 ]
}
