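# Translucent, unlit, two-sided Niagara material for the sticky drift trail vehicle cosmetic, colored by particle color through an emissive adjustment function.
# Opacity multiplies a powered panning cloud noise, a scrolling foam gradient and a U-axis fade at both ends, then particle alpha and depth fade.
# A time-driven sine over the second UV set, scaled by strength and a U-based power ramp, builds a Z-only wave offset vector.
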Material /VehicleCosmetics/DriftTrail/DriftTrail_Sticky/Materials/Cone2_MAT
graph {
 properties {
  bAutomaticallySetUsageInEditor=false
  bUsedWithNiagaraMeshParticles=true
  bUsedWithNiagaraRibbons=true
  bUsedWithNiagaraSprites=true
  blendMode=BLEND_Translucent
  shadingModel=MSM_Unlit
  twoSided=true
  usedShadingModels=Unlit
 }
 nodes [
  n1 = Constant()
  n2 = ComponentMask(None=TextureCoordinate() {coordinateIndex=1}) {g=true, r=true}
  n3 = ScalarParameter() {defaultValue=1, parameterName=CountWaves2}
  n4 = ScalarParameter() {defaultValue=1, parameterName=Speed}
  n5 = OneMinus(None=Time())
  n6 = Sine(None=Add(A=Divide(A=@n2, B=@n3), B=Multiply(A=@n4, B=@n5)))
  n7 = ScalarParameter() {defaultValue=2.5, parameterName=Strenght}
  n8 = Power(Base=ComponentMask(None=TextureCoordinate()) {r=true}, Exp=ScalarParameter() {defaultValue=0.3, parameterName=PowerMiddle2})
  n9 = ScalarParameter() {defaultValue=0.3, parameterName=MultPower2}
  AppendVector(A=AppendVector(A=@n1, B=@n1), B=Multiply(A=Multiply(A=@n6, B=@n7), B=Multiply(A=@n8, B=@n9)))
  MaterialExpressionMeshEmitterVertexColor_0 = ParticleColor()
  MaterialFunctionCall(Emissive=@MaterialExpressionMeshEmitterVertexColor_0.RGB) {materialFunction=&MF_FX_EmissiveAdjustments}
  n10 = TextureSample(UVs=Panner(Coordinate=TextureCoordinate() {uTiling=1.2}) {speedY=-3}) {texture=/DealershipSharedLibrary/FX/Textures/Noise/T_VCVFX_Darkcloud2Tail_RGB}
  n11 = Constant() {r=3}
  n12 = Panner(Coordinate=TextureCoordinate()) {speedX=-0.5}
  n13 = Multiply(
   A=Multiply(
    A=Power(Base=@n10.R, Exp=@n11)
    B=TextureSample(UVs=@n12) {texture=/DealershipSharedLibrary/FX/Textures/Gradient_Masks/T_VCVFX_StickyFoamGradient01_A}.R
   )
   B=Constant() {r=2}
  )
  n14 = ComponentMask(None=TextureCoordinate()) {r=true}
  n15 = Constant() {r=1}
  n16 = OneMinus(None=ComponentMask(None=TextureCoordinate()) {r=true})
  n17 = Constant() {r=1}
  n18 = Multiply(A=Multiply(A=Power(Base=@n14, Exp=@n15), B=Constant() {r=5}), B=Multiply(A=Power(Base=@n16, Exp=@n17), B=Constant() {r=10}))
  MaterialExpressionConstantClamp_3 = Saturate(None=Multiply(A=Multiply(A=@n13, B=@n18), B=@MaterialExpressionMeshEmitterVertexColor_0.A))
  Multiply(
   A=@MaterialExpressionConstantClamp_3
   B=MaterialFunctionCall(Opacity In=Constant() {r=1}, DepthFade Distance=Constant() {r=25}) {materialFunction=&MF_VCVFX_DepthFade}.Result
  )
 ]
}
