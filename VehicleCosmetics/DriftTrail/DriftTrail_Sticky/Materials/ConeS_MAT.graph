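# Unlit, two-sided masked Niagara material for the sticky drift trail vehicle cosmetic, judging by its path and cone name.
# Panning cloud noise distorts scrolling dark-cloud layers, which combine with a rotated tail gradient, foam gradient and UV edge fades into a saturated mask scaled by particle alpha.
# Particle color tints the sharpened noise through an emissive-adjustment function.
# A time-driven sine wave on the second UV channel builds a Z-only offset vector, weighted along U.
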Material /VehicleCosmetics/DriftTrail/DriftTrail_Sticky/Materials/ConeS_MAT
graph {
 properties {
  bAutomaticallySetUsageInEditor=false
  bUsedWithNiagaraMeshParticles=true
  bUsedWithNiagaraRibbons=true
  bUsedWithNiagaraSprites=true
  blendMode=BLEND_Masked
  opacityMaskClipValue=0.333
  shadingModel=MSM_Unlit
  twoSided=true
  usedShadingModels=Unlit
 }
 nodes [
  n1 = Constant()
  n2 = ComponentMask(None=TextureCoordinate() {coordinateIndex=1}) {r=true}
  n3 = ScalarParameter() {defaultValue=0.5, parameterName=CountWaves2}
  n4 = ScalarParameter() {defaultValue=1, parameterName=Speed}
  n5 = OneMinus(None=Time())
  n6 = Sine(None=Add(A=Divide(A=@n2, B=@n3), B=Multiply(A=@n4, B=@n5)))
  n7 = ScalarParameter() {defaultValue=3, parameterName=Strenght}
  n8 = Power(Base=ComponentMask(None=TextureCoordinate()) {r=true}, Exp=ScalarParameter() {defaultValue=2, parameterName=PowerMiddle2})
  n9 = ScalarParameter() {defaultValue=0.5, parameterName=MultPower2}
  AppendVector(A=AppendVector(A=@n1, B=@n1), B=Multiply(A=Multiply(A=@n6, B=@n7), B=Multiply(A=@n8, B=@n9)))
  n10 = Panner(Coordinate=TextureCoordinate() {uTiling=0.35, vTiling=0.35}, Time=Multiply(A=Time(), B=Constant() {r=0.8})) {speedY=-0.2}
  n11 = Multiply(
   A=ComponentMask(None=TextureSample(UVs=@n10) {texture=/DealershipSharedLibrary/FX/Textures/Noise/T_VCVFX_CloudsSoft_RGB}.RGB) {g=true, r=true}
   B=ScalarParameter() {defaultValue=0.5, parameterName=Distortion}
  )
  n12 = TextureCoordinate()
  n13 = TextureSample(UVs=Panner(Coordinate=Add(A=@n11, B=@n12)) {speedY=-2}) {texture=/DealershipSharedLibrary/FX/Textures/Noise/T_VCVFX_Darkcloud2Tail_RGB}
  n14 = OneMinus(None=ComponentMask(None=TextureCoordinate()) {g=true})
  n15 = Constant() {r=2}
  n16 = Multiply(A=@n13.R, B=Multiply(A=Power(Base=@n14, Exp=@n15), B=Constant() {r=5}))
  n17 = TextureCoordinate()
  n18 = AppendVector(A=ScalarParameter() {defaultValue=1, parameterName=X}, B=ScalarParameter() {defaultValue=1, parameterName=Y})
  n19 = MaterialFunctionCall(UVs=Panner(Coordinate=Multiply(A=@n17, B=@n18)), Rotation Center=Constant() {r=0.5}, "Rotation Angle (0-1)"=Constant() {r=-0.25}) {materialFunction=&CustomRotator}
  n20 = Panner(Coordinate=TextureCoordinate()) {speedX=0.5}
  n21 = Multiply(
   A=Multiply(
    A=TextureSample(UVs=@n19.Rotated Values) {texture=/DealershipSharedLibrary/FX/Textures/Gradient_Masks/T_VCVFX_DarkcloudTail_A}.RGB
    B=Constant() {r=5}
   )
   B=Power(
    Base=TextureSample(UVs=@n20) {texture=/DealershipSharedLibrary/FX/Textures/Gradient_Masks/T_VCVFX_StickyFoamGradient01_A}.RGB
    Exp=Constant() {r=1}
   )
  )
  n22 = Constant()
  n23 = Constant() {r=0.5}
  n24 = TextureCoordinate()
  n25 = Multiply(A=Power(Base=ComponentMask(None=@n24) {g=true}, Exp=Constant() {r=2}), B=Constant() {r=5})
  n26 = Time()
  n27 = Constant() {r=0.8}
  n28 = TextureSample(UVs=Panner(Coordinate=TextureCoordinate() {uTiling=0.35, vTiling=0.35}, Time=Multiply(A=@n26, B=@n27)) {speedY=-0.2}) {texture=&T_VCVFX_CloudsSoft_RGB}
  n29 = Add(
   A=Multiply(A=ComponentMask(None=@n28.RGB) {g=true, r=true}, B=ScalarParameter() {defaultValue=0.5, parameterName=Distortion})
   B=TextureCoordinate() {uTiling=2}
  )
  n30 = Power(
   Base=TextureSample(UVs=Panner(Coordinate=@n29) {speedY=-0.5}) {texture=&T_VCVFX_Darkcloud2Tail_RGB}.RGB
   Exp=Constant() {r=2}
  )
  n31 = Constant() {r=1}
  n32 = Multiply(A=Power(Base=@n13.R, Exp=Constant() {r=3}), B=Constant() {r=2})
  n33 = Multiply(A=Add(A=Multiply(A=@n30, B=@n31), B=@n32), B=Multiply(A=Multiply(A=@n16, B=@n21), B=SmoothStep(Min=@n22, Max=@n23, Value=@n25)))
  n34 = ComponentMask(None=TextureCoordinate()) {r=true}
  n35 = Constant() {r=1}
  n36 = OneMinus(None=ComponentMask(None=TextureCoordinate()) {r=true})
  n37 = Constant() {r=1}
  n38 = Multiply(A=Multiply(A=Power(Base=@n34, Exp=@n35), B=Constant() {r=5}), B=Multiply(A=Power(Base=@n36, Exp=@n37), B=Constant() {r=5}))
  MaterialExpressionMeshEmitterVertexColor_1 = ParticleColor()
  MaterialExpressionConstantClamp_4 = Saturate(None=Multiply(A=Multiply(A=@n33, B=@n38), B=@MaterialExpressionMeshEmitterVertexColor_1.A))
  MaterialFunctionCall(Emissive=Multiply(A=@MaterialExpressionMeshEmitterVertexColor_1.RGB, B=@n32)) {materialFunction=&MF_FX_EmissiveAdjustments}
  MaterialFunctionCall!(Input=@n25, Amount=@n22, Smoothness=@n23) {materialFunction=/BRCosmetics/Weapons/FORT_Melee/Pickaxe_Dazzle/FX/MF_Smoothstep}
 ]
}
