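# Masked unlit Niagara material for tentacle trails in the sticky drift trail cosmetic.
# Panned cloud noise distorts the tentacle texture, whose alpha fades in along U for masking, while its desaturated RGB, tinted by particle color, feeds emissive adjustments.
# Blended time-driven sine and cosine waves form WPO-style wobble weighted along U.
# SmallorBigTentacles switches that wobble's transform into view space between two source spaces.
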Material /VehicleCosmetics/DriftTrail/DriftTrail_Sticky/Materials/TentaTrails_MAT
graph {
 properties {
  bAutomaticallySetUsageInEditor=false
  bUsedWithNiagaraMeshParticles=true
  bUsedWithNiagaraRibbons=true
  bUsedWithNiagaraSprites=true
  blendMode=BLEND_Masked
  shadingModel=MSM_Unlit
  usedShadingModels=Unlit
 }
 nodes [
  n1 = Constant()
  n2 = ScalarParameter() {defaultValue=1, parameterName=FadeIn}
  n3 = TextureCoordinate()
  n4 = Multiply(A=Power(Base=ComponentMask(None=@n3) {r=true}, Exp=Constant() {r=1}), B=Constant() {r=10})
  n5 = TextureCoordinate() {uTiling=0.35, vTiling=0.35}
  n6 = Multiply(A=Time(), B=Constant() {r=0.8})
  n7 = ComponentMask(None=TextureSample(UVs=Panner(Coordinate=@n5, Time=@n6) {speedX=-0.1, speedY=0.1}) {texture=/DealershipSharedLibrary/FX/Textures/Noise/T_VCVFX_CloudsSoft_RGB}.RGB) {g=true, r=true}
  n8 = ScalarParameter() {defaultValue=0.1, parameterName=Distortion}
  n9 = TextureSample(UVs=Add(A=Multiply(A=@n7, B=@n8), B=TextureCoordinate())) {texture=/DealershipSharedLibrary/FX/Textures/Gradient_Masks/T_VCVFX_Sticky_TentaTrail_A}
  MaterialExpressionConstantClamp_1 = Saturate(None=Multiply(A=@n9.A, B=SmoothStep(Min=@n1, Max=@n2, Value=@n4)))
  MaterialExpressionVertexColor_1 = ParticleColor()
  n10 = Multiply(A=Constant() {r=2}, B=Multiply(A=@MaterialExpressionVertexColor_1.RGB, B=Desaturation(None=@n9.RGB)))
  MaterialFunctionCall(Emissive=SmoothStep(Value=Saturate(None=@n10)) {constMax=0.65}) {materialFunction=&MF_FX_EmissiveAdjustments}
  n11 = ScalarParameter() {defaultValue=1, parameterName=SmallorBigTentacles}
  n12 = Multiply(A=OneMinus(None=Time()), B=ScalarParameter() {defaultValue=1, group=WPO, parameterName=Speed})
  n13 = Divide(A=ComponentMask(None=TextureCoordinate()) {r=true}, B=ScalarParameter() {defaultValue=1, group=WPO, parameterName=tile})
  n14 = Add(A=@n12, B=@n13)
  n15 = Add(A=@n13, B=@n12)
  n16 = AppendVector(A=AppendVector(A=Cosine(None=@n14) {period=0.5}, B=Sine(None=@n15)), B=Constant())
  n17 = Sine(None=@n14) {period=0.75}
  n18 = Cosine(None=@n15)
  n19 = AppendVector(A=AppendVector(A=OneMinus(None=@n17), B=OneMinus(None=@n18)), B=Constant())
  MaterialExpressionConstantClamp_2 = Saturate(None=LinearInterpolate(A=Sine(None=@n15) {period=2}, B=Cosine(None=@n14) {period=2.5}, Alpha=Constant() {r=0.5}))
  n20 = Multiply(
   A=Multiply(
    A=LinearInterpolate(A=@n16, B=@n19, Alpha=@MaterialExpressionConstantClamp_2)
    B=ScalarParameter() {defaultValue=5, group=WPO, parameterName=Intensity}
   )
   B=ScalarParameter() {defaultValue=1, group=WPO, parameterName=Amount}
  )
  n21 = Transform(None=@n20) {transformType=TRANSFORM_View}
  n22 = Transform(None=@n20) {transformSourceType=TRANSFORMSOURCE_Local, transformType=TRANSFORM_View}
  n23 = TextureCoordinate()
  Multiply(
   A=ConstantBiasScale(None=If(A=@n11, "A > B"=@n21, "A == B"=@n22, "A < B"=@n22))
   B=Power(Base=ComponentMask(None=@n23) {r=true}, Exp=Constant() {r=0.7})
  )
 ]
}
