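# Unlit translucent Niagara material for a wheel effect in the Sticky drift trail cosmetic, usable on sprites, ribbons and mesh particles.
# Two rotated, scrolling dark-cloud tail layers are UV-distorted by panning soft-cloud noise at differing speeds.
# Their sum is masked by a fade at both V ends, a panning X/Y-scalable tail texture and particle alpha, then saturated.
# Particle color tints the second layer's green channel, feeding the shared emissive-adjustment function.
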Material /VehicleCosmetics/DriftTrail/DriftTrail_Sticky/Materials/Wheel_2_MAT
graph {
 properties {
  bAutomaticallySetUsageInEditor=false
  bUsedWithNiagaraMeshParticles=true
  bUsedWithNiagaraRibbons=true
  bUsedWithNiagaraSprites=true
  blendMode=BLEND_Translucent
  shadingModel=MSM_Unlit
  usedShadingModels=Unlit
 }
 nodes [
  n1 = Panner(
   Coordinate=MaterialFunctionCall(UVs=TextureCoordinate() {uTiling=0.35, vTiling=0.35}, "Rotation Angle (0-1)"=Constant() {r=0.25}) {materialFunction=&CustomRotator}.Rotated Values
   Time=Multiply(A=Time(), B=Constant() {r=0.8})
  ) {speedY=-0.3}
  n2 = Multiply(
   A=ComponentMask(None=TextureSample(UVs=@n1) {texture=/DealershipSharedLibrary/FX/Textures/Noise/T_VCVFX_CloudsSoft_RGB}.RGB) {g=true, r=true}
   B=ScalarParameter() {defaultValue=0.5, parameterName=Distortion}
  )
  n3 = MaterialFunctionCall(UVs=TextureCoordinate(), "Rotation Angle (0-1)"=Constant() {r=0.25}) {materialFunction=&CustomRotator}
  n4 = TextureSample(UVs=Panner(Coordinate=Add(A=@n2, B=@n3.Rotated Values)) {speedY=-1}) {texture=/DealershipSharedLibrary/FX/Textures/Noise/T_VCVFX_Darkcloud2Tail_RGB}
  n5 = Constant() {r=1}
  n6 = Panner(
   Coordinate=MaterialFunctionCall(UVs=TextureCoordinate() {uTiling=0.35, vTiling=0.35}, "Rotation Angle (0-1)"=Constant() {r=0.25}) {materialFunction=&CustomRotator}.Rotated Values
   Time=Multiply(A=Time(), B=Constant() {r=0.8})
  ) {speedY=-0.2}
  n7 = Multiply(
   A=ComponentMask(None=TextureSample(UVs=@n6) {texture=&T_VCVFX_CloudsSoft_RGB}.RGB) {g=true, r=true}
   B=ScalarParameter() {defaultValue=0.5, parameterName=Distortion}
  )
  n8 = MaterialFunctionCall(UVs=TextureCoordinate(), "Rotation Angle (0-1)"=Constant() {r=0.25}) {materialFunction=&CustomRotator}
  n9 = TextureSample(UVs=Panner(Coordinate=Add(A=@n7, B=@n8.Rotated Values)) {speedY=-1.5}) {texture=&T_VCVFX_Darkcloud2Tail_RGB}
  n10 = Multiply(A=Power(Base=@n9.G, Exp=Constant() {r=2}), B=Constant() {r=2})
  n11 = Add(A=Multiply(A=Power(Base=@n4.RGB, Exp=@n5), B=Constant() {r=2}), B=@n10)
  n12 = TextureCoordinate()
  n13 = Multiply(A=Power(Base=ComponentMask(None=@n12) {g=true}, Exp=Constant() {r=2}), B=Constant() {r=5})
  n14 = ComponentMask(None=TextureCoordinate()) {g=true}
  n15 = Multiply(A=Power(Base=OneMinus(None=@n14), Exp=Constant() {r=2}), B=Constant() {r=5})
  n16 = ScalarParameter() {defaultValue=1, parameterName=X}
  n17 = ScalarParameter() {defaultValue=1, parameterName=Y}
  n18 = Panner(Coordinate=Multiply(A=TextureCoordinate() {uTiling=3}, B=AppendVector(A=@n16, B=@n17))) {speedY=-1}
  n19 = Multiply(
   A=Multiply(A=@n9.R, B=Multiply(A=@n13, B=@n15))
   B=Multiply(A=TextureSample(UVs=@n18) {texture=/DealershipSharedLibrary/FX/Textures/Gradient_Masks/T_VCVFX_DarkcloudTail_A}.RGB, B=Constant() {r=5})
  )
  MaterialExpressionMeshEmitterVertexColor_1 = ParticleColor()
  MaterialExpressionConstantClamp_4 = Saturate(None=Multiply(A=Multiply(A=@n11, B=@n19), B=@MaterialExpressionMeshEmitterVertexColor_1.A))
  MaterialFunctionCall(Emissive=Multiply(A=@MaterialExpressionMeshEmitterVertexColor_1.RGB, B=@n10)) {materialFunction=&MF_FX_EmissiveAdjustments}
 ]
}
