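# Translucent, unlit, two-sided Niagara material for the wheel effect of the Sticky drift trail cosmetic.
# A tiled dark-cloud noise texture, rotated a quarter turn and panned along V, is sharpened by a power and boosted.
# A V-gradient times its inverse fades both edges, and particle alpha scales the clamped mask.
# Particle color feeds the shared emissive adjustment function.
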Material /VehicleCosmetics/DriftTrail/DriftTrail_Sticky/Materials/Wheel_MAT
graph {
 properties {
  bAutomaticallySetUsageInEditor=false
  bUsedWithNiagaraMeshParticles=true
  bUsedWithNiagaraRibbons=true
  bUsedWithNiagaraSprites=true
  blendMode=BLEND_Translucent
  shadingModel=MSM_Unlit
  twoSided=true
  usedShadingModels=Unlit
 }
 nodes [
  n1 = TextureCoordinate() {uTiling=3}
  n2 = Constant() {r=0.25}
  n3 = TextureSample(UVs=Panner(Coordinate=MaterialFunctionCall(UVs=@n1, "Rotation Angle (0-1)"=@n2) {materialFunction=&CustomRotator}.Rotated Values) {speedY=-2}) {texture=/DealershipSharedLibrary/FX/Textures/Noise/T_VCVFX_Darkcloud2Tail_RGB}
  n4 = Constant() {r=3}
  n5 = Multiply(A=Multiply(A=Power(Base=@n3.RGB, Exp=@n4)), B=Constant() {r=2})
  n6 = ComponentMask(None=TextureCoordinate()) {g=true}
  n7 = Constant() {r=1.5}
  n8 = OneMinus(None=ComponentMask(None=TextureCoordinate()) {g=true})
  n9 = Constant() {r=1.5}
  n10 = Multiply(A=Multiply(A=Power(Base=@n6, Exp=@n7), B=Constant() {r=5}), B=Multiply(A=Power(Base=@n8, Exp=@n9), B=Constant() {r=5}))
  MaterialExpressionMeshEmitterVertexColor_0 = ParticleColor()
  MaterialExpressionConstantClamp_3 = Saturate(None=Multiply(A=Multiply(A=@n5, B=@n10), B=@MaterialExpressionMeshEmitterVertexColor_0.A))
  MaterialFunctionCall(Emissive=@MaterialExpressionMeshEmitterVertexColor_0.RGB) {materialFunction=&MF_FX_EmissiveAdjustments}
 ]
}
